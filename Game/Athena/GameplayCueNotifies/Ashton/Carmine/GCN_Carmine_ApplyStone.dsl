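; Burst gameplay cue for applying a Carmine stone, playing a colored pickup wash effect on the target player pawn.
; The cue's raw magnitude is truncated to an integer and used as the key into the ColorMapping variable.
; When a color is found, it spawns the Cascade wash emitter attached to the pawn mesh's root and passes that color in as the Color vector parameter.
; An unmapped key spawns nothing.
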(fn OnBurst(Cascade)
  (params (MyTarget :type Actor%20Object%20Reference) (Parameters :type Gameplay%20Cue%20Parameters%20Structure%20%28by%20ref%29) (ParticleComponents :type Array%20of%20Cascade%20Particle%20System%20Component%20Object%20References) (AudioComponents :type Array%20of%20Audio%20Component%20Object%20References) (BurstCameraShakeInstance :type Camera%20Shake%20Base%20Object%20Reference) (BurstDecalInstance :type Decal%20Actor%20Object%20Reference))
  (bind (_normalizedmagnitude _rawmagnitude) (Ability|GameplayCue|BreakGameplayCueParameters Parameters))
  (bind _returnvalue (Math|Float|Truncate _rawmagnitude))
  (bind _colormapping (Variables|Default|GetColorMapping))
  (bind (_value _returnvalue_2) (Utilities|Map|Find _colormapping _returnvalue))
  (bind _asfort_player_pawn (Utilities|Casting|CastToFortPlayerPawn MyTarget))
  (if _returnvalue_2
    (bind _returnvalue_3 (Effects|SpawnEmitterAttached "/Game/Athena/Items/Gameplay/BackPacks/Ashton/Carmine/P_Carmine_Pickup_Wash.P_Carmine_Pickup_Wash" (Variables|/Script/Engine.Character|GetMesh _asfort_player_pawn) "root"))
    (Effects|Components|ParticleSystem|SetVectorParameter _returnvalue_3 "Color" (Math|Conversions|ToVector(LinearColor) _value))
    (return)))
